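# World marker indicator widget for pinned quests, built on AthenaQuestMarkedActorIndicator with a visibility switcher holding default and custom pin images.
# When marker data is set, the pinned quest index drives the BackingTypeSelector scalar on the default pin's dynamic material.
# ShowHide fades the default pin in on construct and plays reversed on removal notice, with removal duration taken from the animation's end time.
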WidgetBlueprint /Game/Athena/UI/Challenges/W_PinnedQuestMarkerIndicator
animations [ShowHide {
  binding_count=1
  length_seconds=0.4
  object=&self:ShowHide
  sequence {bindings [Image_PinDefault {
     id=#1
     possessed_class=Image
     tracks [{
       type=MovieSceneFloatTrack
       display_name=Render Opacity
       sections [{type=MovieSceneFloatSection, end_frame=8, start_frame=0, channels [None {type=MovieSceneScriptingFloatChannel, keys [[0,0],[7.9997,1]]}]}]
      }]
    }]}
 }]
graphs [
 EventGraph {nodes [
   Event Construct()
   > PlayAnimation(InAnimation=Get ShowHide() -> (ShowHide:Widget Animation Object Reference))
   n1 = Event OnSetMarkerData!() -> (OutputDelegate, then, FortWorldMarkerData:"Fort World Marker Data Structure (by ref)") {is=title K2Node_Event}
   > n2 = GetDynamicMaterial(self=Get Image_PinDefault() -> (Image_PinDefault:Image Object Reference))
   n3 = Break Fort World Marker Data!(FortWorldMarkerData:"Fort World Marker Data Structure (by ref)"=@n1.FortWorldMarkerData) -> (PinnedQuestIndex:int) {is=title K2Node_BreakStruct}
   > SetScalarParameterValue(self=@n2, ParameterName=BackingTypeSelector, Value="To Float (Integer)"(InInt=int + int(A=@n3)))
   Event PreConstruct()
  ]}
 GetCanvasPanelContainer {nodes [
   GetCanvasPanelContainer!() -> (then, ContainerEnum:EFortActorIndicatorContainerWidget Enum=Top) {is=title K2Node_FunctionEntry}
   > Return Node!(execute, ReturnValue:Canvas Panel Object Reference)
  ]}
 GetIndicatedActor {nodes [
   GetIndicatedActor() {is=title K2Node_FunctionEntry}
   > Return Node(ReturnValue:Actor Object Reference)
  ]}
 GetRemovalDuration {nodes [
   GetRemovalDuration() {is=title K2Node_FunctionEntry}
   > Return Node(ReturnValue:float=GetEndTime(self=Get ShowHide()))
  ]}
 NotifyWillBeRemoved {nodes [
   n1 = GetDynamicMaterial(self=Get Pin() -> (Pin:Image Object Reference))
   n2 = Break Fort World Marker Data()
   > SetScalarParameterValue(self=@n1, ParameterName=BackingTypeSelector, Value="To Float (Integer)"(InInt=int + int(A=@n2)))
   NotifyWillBeRemoved() {is=title K2Node_FunctionEntry}
   > PlayAnimation(InAnimation=Get ShowHide(), PlayMode=Reverse)
   > Return Node!(execute)
  ]}
]
variables []
widgets {
 info {inheritedWidgetCount=8, parentClass=/Script/FortniteUI.AthenaQuestMarkedActorIndicator, rootWidgetClass=&Overlay, widgetCount=9}
 widgets [
  Overlay_19 {widgetClassPath=&Overlay, widget=Overlay_19}
  Switcher_PinImages {
   widgetClassPath=&CommonVisibilitySwitcher
   bInherited=true
   parent=Overlay_19
   widget=Switcher_PinImages
   slot {class=OverlaySlot}
  }
  Image_PinDefault {
   widgetClassPath=&Image
   bInherited=true
   bIsVariable=true
   parent=Switcher_PinImages
   widget=Image_PinDefault
   properties {brush {
     bIsDynamicallyLoaded=false
     drawAs=Image
     imageType=FullColor
     mirroring=NoMirror
     resourceName=None
     resourceObject=&/MI_UI_QuestPin_WorldMarker
     tiling=NoTile
     imageSize=<64 64>
     margin=<ltrb 0 0 0 0>
     outlineSettings {
      bUseBrushTransparency=false
      roundingType=HalfHeightRadius
      width=0
      color {colorUseRule=UseColor_Specified, specifiedColor=<rgba 0 0 0 0>}
      cornerRadii=<0 0 0 0>
     }
     tintColor {colorUseRule=UseColor_Specified, specifiedColor=<rgba 1 1 1 1>}
     uVRegion {bIsValid=false, max=<0 0>, min=<0 0>}
    }}
   slot {class=CommonVisibilitySwitcherSlot}
  }
  LazyImage_PinCustom {
   widgetClassPath=&FortLazyImage
   bInherited=true
   bIsVariable=true
   parent=Switcher_PinImages
   widget=LazyImage_PinCustom
   properties {brush {
     bIsDynamicallyLoaded=false
     drawAs=Image
     imageType=NoImage
     mirroring=NoMirror
     resourceName=None
     resourceObject=null
     tiling=NoTile
     imageSize=<64 64>
     margin=<ltrb 0 0 0 0>
     outlineSettings {
      bUseBrushTransparency=false
      roundingType=HalfHeightRadius
      width=0
      color {colorUseRule=UseColor_Specified, specifiedColor=<rgba 0 0 0 0>}
      cornerRadii=<0 0 0 0>
     }
     tintColor {colorUseRule=UseColor_Specified, specifiedColor=<rgba 1 1 1 1>}
     uVRegion {bIsValid=false, max=<0 0>, min=<0 0>}
    }}
   slot {class=CommonVisibilitySwitcherSlot}
  }
  Switcher_IndicatorSize {widgetClassPath=&WidgetSwitcher, bInherited=true, widget=null}
  Panel_SmallIndicator {widgetClassPath=/Script/UMG.PanelWidget, bInherited=true, widget=null}
  Panel_LargeIndicator {widgetClassPath=&PanelWidget, bInherited=true, widget=null}
  Text_Distance {widgetClassPath=&CommonTextBlock, bInherited=true, widget=null}
  TutorialHighlight {widgetClassPath=&TutorialHighlight}
 ]
}
