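# Punchcard_PlaceStamp_Cue is a sound cue for the quest punchcard, named for the placing of a stamp.
# It routes to the non-critical in-game UI sound class at reduced volume.
# Its own concurrency override allows one instance at a time, stopping the oldest.
# The cue shows no root or sound nodes, so it references no audio here.
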SoundCue /Game/Athena/Sounds/Quests/Punchcard/Punchcard_PlaceStamp_Cue
root=null
nodes []
properties {
 bOverrideConcurrency=true
 maxDistance=2097152
 soundClassObject=&Fort_UI_InGame_NonCritical
 volumeMultiplier=0.7
 concurrencyOverrides {maxCount=1, resolutionRule=StopOldest}
}
