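# UI material for the Hightower logo, stored with the Athena challenges UI art.
# It uses the UI domain with translucent blending and contains a texture sample of the 128 logo texture and a constant node.
# The graph lists no connections from these nodes to the material outputs.
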Material /Game/Athena/UI/Challenges/Art/T_Ui_Hightower_Logo_128_Mat
graph {
 properties {
  bUsedWithStaticMesh=false
  blendMode=BLEND_Translucent
  materialDomain=MD_UI
  pixelDepthOffsetMode=PDOM_Legacy
  usedShadingModels=Default Lit
 }
 nodes [
  Constant() {r=1}
  TextureSample() {texture=&/T_Ui_Hightower_Logo_128}
 ]
}
