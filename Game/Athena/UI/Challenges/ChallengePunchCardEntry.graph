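# Widget for one challenge punch card slot, showing an empty socket, linking connectors, a stamp and an unlock date hidden while in zone.
# SetStyle recolors these parts, partly through dynamic materials, and first, last, default and nonsequential slot events toggle connectors and socket parameters.
# PreConstruct randomizes stamp scale, vertical offset and angle for a hand-stamped look.
# OnPlacedEntry plays the PlaceStamp animation, then signals completion after a delay matching its length.
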WidgetBlueprint /Game/Athena/UI/Challenges/ChallengePunchCardEntry
animations [PlaceStamp {
  binding_count=4
  length_seconds=1.6667
  object=&self:PlaceStamp
  sequence {
   bindings [
    Image_Flare {
     id=#1
     possessed_class=Image
     tracks [
      {
       type=MovieScene2DTransformTrack
       sections [{
         type=MovieScene2DTransformSection
         channels [
          Translation.X {type=MovieSceneScriptingFloatChannel}
          Translation.Y {type=MovieSceneScriptingFloatChannel}
          Angle {type=MovieSceneScriptingFloatChannel}
          Scale.X {type=MovieSceneScriptingFloatChannel, thinned_from=4, keys [[57,0.5],[75,2.6],[98,2]]}
          Scale.Y {type=MovieSceneScriptingFloatChannel, thinned_from=4, keys [[57,0.5],[75,2.6],[98,2]]}
          Shear.X {type=MovieSceneScriptingFloatChannel}
          Shear.Y {type=MovieSceneScriptingFloatChannel}
         ]
        }]
      }
      {
       type=MovieSceneFloatTrack
       display_name=Render Opacity
       sections [{type=MovieSceneFloatSection, channels [None {type=MovieSceneScriptingFloatChannel, keys [[57,0],[61.459,1],[100,0]]}]}]
      }
     ]
    }
    LazyImage_Stamp {
     id=#2
     possessed_class=FortLazyImage
     tracks [
      {
       type=MovieScene2DTransformTrack
       sections [{
         type=MovieScene2DTransformSection
         channels [
          Translation.X {type=MovieSceneScriptingFloatChannel}
          Translation.Y {type=MovieSceneScriptingFloatChannel}
          Angle {type=MovieSceneScriptingFloatChannel}
          Scale.X {type=MovieSceneScriptingFloatChannel, keys [[15,10],[33,2.5],[48,2.4],[57,1],[65,1.625],[66,1.9],[80,1.6]]}
          Scale.Y {type=MovieSceneScriptingFloatChannel, keys [[15,10],[33,2.5],[48,2.4],[57,1.2],[65,1.625],[66,1.9],[80,1.6]]}
          Shear.X {type=MovieSceneScriptingFloatChannel}
          Shear.Y {type=MovieSceneScriptingFloatChannel}
         ]
        }]
      }
      {
       type=MovieSceneFloatTrack
       display_name=Render Opacity
       sections [{type=MovieSceneFloatSection, channels [None {type=MovieSceneScriptingFloatChannel, keys [[0,0],[15,0],[30,1]]}]}]
      }
     ]
    }
    Image_Spark {
     id=#3
     possessed_class=Image
     tracks [
      {
       type=MovieSceneFloatTrack
       display_name=Render Opacity
       sections [{
         type=MovieSceneFloatSection
         end_frame=78
         start_frame=56
         channels [None {type=MovieSceneScriptingFloatChannel, keys [[57,0],[63,1],[75,0]]}]
        }]
      }
      {
       type=MovieScene2DTransformTrack
       sections [{
         type=MovieScene2DTransformSection
         end_frame=78
         start_frame=56
         channels [
          Translation.X {type=MovieSceneScriptingFloatChannel}
          Translation.Y {type=MovieSceneScriptingFloatChannel}
          Angle {type=MovieSceneScriptingFloatChannel}
          Scale.X {type=MovieSceneScriptingFloatChannel, keys [[57,0.75],[75,2]]}
          Scale.Y {type=MovieSceneScriptingFloatChannel, keys [[57,0.75],[75,2]]}
          Shear.X {type=MovieSceneScriptingFloatChannel}
          Shear.Y {type=MovieSceneScriptingFloatChannel}
         ]
        }]
      }
     ]
    }
    Text_UnlockDate {
     id=#4
     possessed_class=CommonTextBlock
     tracks [
      {
       type=MovieScene2DTransformTrack
       sections [{
         type=MovieScene2DTransformSection
         channels [
          Translation.X {type=MovieSceneScriptingFloatChannel, const [0,1]}
          Translation.Y {type=MovieSceneScriptingFloatChannel, const [0,1]}
          Angle {type=MovieSceneScriptingFloatChannel}
          Scale.X {type=MovieSceneScriptingFloatChannel, keys [[48,4],[57,0.9],[68.987,1]]}
          Scale.Y {type=MovieSceneScriptingFloatChannel, keys [[48,4],[57,0.9],[68.987,1]]}
          Shear.X {type=MovieSceneScriptingFloatChannel}
          Shear.Y {type=MovieSceneScriptingFloatChannel}
         ]
        }]
      }
      {
       type=MovieSceneFloatTrack
       display_name=Render Opacity
       sections [{type=MovieSceneFloatSection, channels [None {type=MovieSceneScriptingFloatChannel, keys [[48,0],[57,1]]}]}]
      }
     ]
    }
   ]
   display_rate {numerator=60}
  }
 }]
graphs [EventGraph {nodes [
   n1 = Event SetStyle!() -> (OutputDelegate, then, PunchStyle:"Fort Challenge Punch Card Styles Structure (by ref)") {is=title K2Node_Event}
   n2 = Break Fort Challenge Punch Card Styles!(FortChallengePunchCardStyles:"Fort Challenge Punch Card Styles Structure (by ref)"=@n1.PunchStyle) -> (
    EncodedName:String
    BaseColor1:Linear Color Structure
    BaseColor2:Linear Color Structure
    BaseColor3:Linear Color Structure
    BaseColor4:Linear Color Structure
    BaseColor5:Linear Color Structure
    AccentColor1:Linear Color Structure
    AccentColor2:Linear Color Structure
    BackgrounTexture:Texture 2D Soft Object Reference
   ) {is=title K2Node_BreakStruct}
   SetColorAndOpacity(
    self:Image Object Reference=[
     Get Image_ConnectorLeft() -> (Image_ConnectorLeft:Image Object Reference)
     Get Image_ConnectorRight() -> (Image_ConnectorRight:Image Object Reference)
     Get Image_EmptySocket() -> (Image_EmptySocket:Image Object Reference)
    ]
    InColorAndOpacity:Linear Color Structure=@n2.BaseColor5
   )
   > SetContentColorAndOpacity!(
    execute
    self:Border Object Reference=Get Border_Date() -> (Border_Date:Border Object Reference)
    InContentColorAndOpacity:Linear Color Structure=@n2.BaseColor2
   ) -> (then)
   > n3 = GetDynamicMaterial(self=Get LazyImage_Stamp() -> (LazyImage_Stamp:Fort Lazy Image Object Reference))
   > SetVectorParameterValue(self=@n3, ParameterName=BaseColor, Value=@n2.BaseColor2)
   > SetVectorParameterValue(self=@n3, ParameterName=Accent01, Value=@n2.AccentColor1)
   > SetVectorParameterValue(self=@n3, ParameterName=Accent02, Value=@n2.AccentColor2)
   > n4 = GetDynamicMaterial(self=Get Image_Flare() -> (Image_Flare:Image Object Reference))
   > SetVectorParameterValue(self=@n4, ParameterName=Rarity Color 1, Value=@n2.BaseColor2)
   > SetVectorParameterValue(self=@n4, ParameterName=Rarity Color 3, Value=@n2.BaseColor3)
   Play Place Stamp() {is=title K2Node_CustomEvent}
   > PlaySound2D(Sound=/Game/Athena/Sounds/Quests/Punchcard/Punchcard_PlaceStamp_Cue)
   Event PreConstruct()
   n5 = RandomFloatInRange(Min=Get Min Scale() -> (Min Scale:double), Max=Get Max Scale() -> (Max Scale:double))
   n6 = Get Overlay_Stamp() -> (Overlay_Stamp:Overlay Object Reference)
   > SetRenderScale!(execute, self:Widget Object Reference=@n6, Scale:Vector 2D Structure=MakeVector2D(X=@n5, Y=@n5)) -> (then)
   n7 = Get Min Translation() -> (Min Translation:double)
   n8 = Get Max Translation() -> (Max Translation:double)
   > SetRenderTranslation(self=@n6, Translation=MakeVector2D(X=0, Y=RandomFloatInRange(Min=@n7, Max=@n8)))
   > SetRenderTransformAngle!(
    execute
    self:Widget Object Reference=@n6
    Angle:float=RandomFloatInRange(Min="Get Min  Rotation"() -> ("Min  Rotation":double), Max=Get maxRotation() -> (maxRotation:double))
   ) -> (then)
   Event Construct()
   n9 = Get FortGlobalUIContext() -> (ReturnValue:Fort Global UIContext Object Reference) {is=title K2Node_GetSubsystem}
   > SetVisibility(
    self=Get Border_Date()
    InVisibility=Select(Option 0=HitTestInvisible, Option 1=Collapsed, Index=IsInZone!(self:Fort Global UIContext Object Reference=@n9) -> (ReturnValue:bool))
   )
   > BindToAnimationStarted!(
    execute
    self:User Widget Object Reference
    Animation:Widget Animation Object Reference=Get PlaceStamp() -> (PlaceStamp:Widget Animation Object Reference)
    Delegate:Delegate=Create Event()
   ) -> (then)
   Event SetIsLastSlot() {is=title K2Node_Event}
   > SetVisibility(self=Get Image_ConnectorRight(), InVisibility=Hidden)
   > n10 = GetDynamicMaterial(self=Get Image_EmptySocket())
   > SetScalarParameterValue(self=@n10, ParameterName=ShowRight, Value=0)
   Event OnPlacedEntry() {is=title K2Node_Event}
   > SetVisibility(self=Get LazyImage_Stamp(), InVisibility=SelfHitTestInvisible)
   n11 = Get PlaceStamp()
   > PlayAnimationForward(InAnimation=@n11)
   > Delay(Duration=GetEndTime(self=@n11))
   > OnFinishedPlacingEntry(self:Athena Challenge Punch Card Entry Object Reference)
   Event SetIsFirstSlot() {is=title K2Node_Event}
   > SetVisibility(self=Get Image_ConnectorLeft(), InVisibility=Hidden)
   > n12 = GetDynamicMaterial(self=Get Image_EmptySocket())
   > SetScalarParameterValue(self=@n12, ParameterName=ShowLeft, Value=0)
   Event SetToDefaultSlot() {is=title K2Node_Event}
   > SetVisibility(self=[Get Image_ConnectorLeft(), Get Image_ConnectorRight()], InVisibility=SelfHitTestInvisible)
   > n13 = GetDynamicMaterial(self=Get Image_EmptySocket())
   > SetScalarParameterValue(self=@n13, ParameterName=ShowLeft, Value=1)
   > SetScalarParameterValue(self=@n13, ParameterName=ShowRight, Value=1)
   Event SetIsNonsequentialSlot() {is=title K2Node_Event}
   > SetVisibility(self=[Get Image_ConnectorLeft(), Get Image_ConnectorRight()], InVisibility=Hidden)
   > n14 = GetDynamicMaterial(self=Get Image_EmptySocket())
   > SetScalarParameterValue(self=@n14, ParameterName=ShowLeft, Value=0)
   > SetScalarParameterValue(self=@n14, ParameterName=ShowRight, Value=0)
   Reroute Node(InputPin=[@n15.then, @n16.else])
   n15 = SetVisibility(execute=@n16.then, self=Get LazyImage_Stamp(), InVisibility=Hidden)
   n16 = Branch(execute=@n1.then, Condition=Get bIsBeingPlaced() -> (bIsBeingPlaced:bool))
  ]}]
variables [
 AccentColor01 {type=LinearColor, default=<rgba 1 0.8535 0 1>}
 AccentColor02 {type=LinearColor, default=<rgba 1 0.2896 0 1>}
 BaseColor01 {type=LinearColor, default=<rgba 0 0.4431 1 1>}
 BaseColor02 {type=LinearColor, default=<rgba 0 0.4431 1 1>}
 Max Scale {type=number, default=1.075}
 Max Translation {type=number, default=6}
 maxRotation {type=number, default=12}
 "Min  Rotation" {type=number, default=-12}
 Min Scale {type=number, default=0.925}
 Min Translation {type=number, default=-6}
]
widgets {
 info {inheritedWidgetCount=3, parentClass=/Script/FortniteUI.AthenaChallengePunchCardEntry, rootWidgetClass=&SizeBox, widgetCount=20}
 widgets [
  SizeBox_Entry {widgetClassPath=&SizeBox, bIsVariable=true, widget=SizeBox_Entry, properties {bOverride_WidthOverride=true, widthOverride=128}}
  Overlay_4 {widgetClassPath=&Overlay, parent=SizeBox_Entry, widget=Overlay_4, slot {class=SizeBoxSlot}}
  Switcher_SlotState {
   widgetClassPath=&WidgetSwitcher
   bIsVariable=true
   parent=Overlay_4
   widget=Switcher_SlotState
   slot {class=OverlaySlot, properties {horizontalAlignment=HAlign_Fill, verticalAlignment=VAlign_Fill}}
  }
  VerticalBox_1 {
   widgetClassPath=&VerticalBox
   parent=Overlay_4
   widget=VerticalBox_1
   slot {class=OverlaySlot, properties {horizontalAlignment=HAlign_Fill, verticalAlignment=VAlign_Center, padding=<ltrb 0 4 0 4>}}
  }
  Overlay_0 {
   widgetClassPath=&Overlay
   parent=VerticalBox_1
   widget=Overlay_0
   slot {class=VerticalBoxSlot, properties {size {sizeRule=Fill, value=1}}}
  }
  ScaleBox_0 {
   widgetClassPath=&ScaleBox
   parent=Overlay_0
   widget=ScaleBox_0
   slot {class=OverlaySlot, properties {horizontalAlignment=HAlign_Fill, verticalAlignment=VAlign_Center}}
  }
  Track {
   widgetClassPath=&HorizontalBox
   parent=ScaleBox_0
   widget=Track
   slot {class=ScaleBoxSlot, properties {horizontalAlignment=HAlign_Fill}}
  }
  Image_ConnectorLeft {
   widgetClassPath=&Image
   bIsVariable=true
   parent=Track
   widget=Image_ConnectorLeft
   properties {
    visibility=SelfHitTestInvisible
    brush {
     bIsDynamicallyLoaded=false
     drawAs=Box
     imageType=NoImage
     mirroring=NoMirror
     resourceName=None
     resourceObject=/Game/Athena/UI/Challenges/Art/M_Ui_Punchcard_slot_Link
     tiling=NoTile
     imageSize=<0 95>
     margin=<ltrb 0 0 0 0>
     outlineSettings {
      bUseBrushTransparency=false
      roundingType=HalfHeightRadius
      width=0
      color {colorUseRule=UseColor_Specified, specifiedColor=<rgba 0 0 0 0>}
      cornerRadii=<0 0 0 0>
     }
     tintColor {colorUseRule=UseColor_Specified, specifiedColor=<rgba 1 1 1 0.5>}
     uVRegion {bIsValid=false, max=<0 0>, min=<0 0>}
    }
    colorAndOpacity=<rgba 0.0262 0 0.0097 1>
   }
   slot {class=HorizontalBoxSlot, properties {size {sizeRule=Fill, value=1}}}
  }
  Overlay_1 {
   widgetClassPath=&Overlay
   parent=Track
   widget=Overlay_1
   slot {class=HorizontalBoxSlot, properties {verticalAlignment=VAlign_Top}}
  }
  Image_EmptySocket {
   widgetClassPath=&Image
   bIsVariable=true
   parent=Overlay_1
   widget=Image_EmptySocket
   properties {
    visibility=SelfHitTestInvisible
    brush {
     bIsDynamicallyLoaded=false
     drawAs=Image
     imageType=NoImage
     mirroring=NoMirror
     resourceName=None
     resourceObject=/Game/Athena/UI/Challenges/Art/M_Ui_Punchcard_slot_Inst
     tiling=NoTile
     imageSize=<110 110>
     margin=<ltrb 0 0 0 0>
     outlineSettings {
      bUseBrushTransparency=false
      roundingType=HalfHeightRadius
      width=0
      color {colorUseRule=UseColor_Specified, specifiedColor=<rgba 0 0 0 0>}
      cornerRadii=<0 0 0 0>
     }
     tintColor {colorUseRule=UseColor_Specified, specifiedColor=<rgba 1 1 1 0.5>}
     uVRegion {bIsValid=false, max=<0 0>, min=<0 0>}
    }
    colorAndOpacity=<rgba 0.0262 0 0.0097 1>
   }
   slot {class=OverlaySlot}
  }
  Overlay_Stamp {
   widgetClassPath=&Overlay
   bIsVariable=true
   parent=Overlay_1
   widget=Overlay_Stamp
   slot {class=OverlaySlot, properties {horizontalAlignment=HAlign_Center, verticalAlignment=VAlign_Center}}
  }
  LazyImage_Stamp {
   widgetClassPath=&FortLazyImage
   bInherited=true
   bIsVariable=true
   parent=Overlay_Stamp
   widget=LazyImage_Stamp
   properties {
    brush {
     bIsDynamicallyLoaded=false
     drawAs=Image
     imageType=NoImage
     mirroring=NoMirror
     resourceName=None
     resourceObject=/Game/Athena/UI/Challenges/Art/M_Ui_PuncardIcon_Inst
     tiling=NoTile
     imageSize=<64 64>
     margin=<ltrb 0 0 0 0>
     outlineSettings {
      bUseBrushTransparency=false
      roundingType=HalfHeightRadius
      width=0
      color {colorUseRule=UseColor_Specified, specifiedColor=<rgba 0 0 0 0>}
      cornerRadii=<0 0 0 0>
     }
     tintColor {colorUseRule=UseColor_Specified, specifiedColor=<rgba 1 1 1 1>}
     uVRegion {bIsValid=false, max=<0 0>, min=<0 0>}
    }
    renderTransform {angle=0, scale=<1.6 1.6>, shear=<0 0>, translation=<0 0>}
   }
   slot {class=OverlaySlot}
  }
  Image_Flare {
   widgetClassPath=&Image
   bIsVariable=true
   parent=Overlay_1
   widget=Image_Flare
   properties {
    renderOpacity=0
    visibility=SelfHitTestInvisible
    brush {
     bIsDynamicallyLoaded=false
     drawAs=Image
     imageType=NoImage
     mirroring=NoMirror
     resourceName=None
     resourceObject=/Game/UI/Frontend/Rewards/materials/M_UI_DailyReward_Rarity-Flares
     tiling=NoTile
     imageSize=<90 90>
     margin=<ltrb 0 0 0 0>
     outlineSettings {
      bUseBrushTransparency=false
      roundingType=HalfHeightRadius
      width=0
      color {colorUseRule=UseColor_Specified, specifiedColor=<rgba 0 0 0 0>}
      cornerRadii=<0 0 0 0>
     }
     tintColor {colorUseRule=UseColor_Specified, specifiedColor=<rgba 1 1 1 1>}
     uVRegion {bIsValid=false, max=<0 0>, min=<0 0>}
    }
    colorAndOpacity=<rgba 1 1 1 0.5>
   }
   slot {class=OverlaySlot, properties {horizontalAlignment=HAlign_Center, verticalAlignment=VAlign_Center, padding=<ltrb -100 -100 -100 -100>}}
  }
  Image_Spark {
   widgetClassPath=&Image
   bIsVariable=true
   parent=Overlay_1
   widget=Image_Spark
   properties {
    renderOpacity=0
    visibility=SelfHitTestInvisible
    brush {
     bIsDynamicallyLoaded=false
     drawAs=Image
     imageType=NoImage
     mirroring=NoMirror
     resourceName=None
     resourceObject=/Game/UI/InGame/Quest/Materials/M_UI_SparkStar
     tiling=NoTile
     imageSize=<90 90>
     margin=<ltrb 0 0 0 0>
     outlineSettings {
      bUseBrushTransparency=false
      roundingType=HalfHeightRadius
      width=0
      color {colorUseRule=UseColor_Specified, specifiedColor=<rgba 0 0 0 0>}
      cornerRadii=<0 0 0 0>
     }
     tintColor {colorUseRule=UseColor_Specified, specifiedColor=<rgba 1 1 1 1>}
     uVRegion {bIsValid=false, max=<0 0>, min=<0 0>}
    }
    colorAndOpacity=<rgba 1 1 1 0.5>
   }
   slot {class=OverlaySlot, properties {horizontalAlignment=HAlign_Center, verticalAlignment=VAlign_Center, padding=<ltrb -100 -100 -100 -100>}}
  }
  Image_ConnectorRight {
   widgetClassPath=&Image
   bIsVariable=true
   parent=Track
   widget=Image_ConnectorRight
   properties {
    visibility=SelfHitTestInvisible
    brush {
     bIsDynamicallyLoaded=false
     drawAs=Box
     imageType=NoImage
     mirroring=NoMirror
     resourceName=None
     resourceObject=&M_Ui_Punchcard_slot_Link
     tiling=NoTile
     imageSize=<0 95>
     margin=<ltrb 0 0 0 0>
     outlineSettings {
      bUseBrushTransparency=false
      roundingType=HalfHeightRadius
      width=0
      color {colorUseRule=UseColor_Specified, specifiedColor=<rgba 0 0 0 0>}
      cornerRadii=<0 0 0 0>
     }
     tintColor {colorUseRule=UseColor_Specified, specifiedColor=<rgba 1 1 1 0.5>}
     uVRegion {bIsValid=false, max=<0 0>, min=<0 0>}
    }
    colorAndOpacity=<rgba 0.0262 0 0.0097 1>
   }
   slot {class=HorizontalBoxSlot, properties {size {sizeRule=Fill, value=1}}}
  }
  VerticalBox_0 {
   widgetClassPath=&VerticalBox
   parent=Overlay_0
   widget=VerticalBox_0
   properties {visibility=Collapsed}
   slot {class=OverlaySlot, properties {horizontalAlignment=HAlign_Center, verticalAlignment=VAlign_Center}}
  }
  Text_RewardQuantity {
   widgetClassPath=&CommonTextBlock
   bInherited=true
   parent=VerticalBox_0
   widget=Text_RewardQuantity
   properties {
    justification=Center
    text=99k
    colorAndOpacity {colorUseRule=UseColor_Specified, specifiedColor=<rgba 0.0003 0.0284 0.1683 1>}
    font {
     bForceMonospaced=false
     bMaterialIsStencil=false
     fontMaterial=null
     fontObject=&BurbankSmall
     letterSpacing=0
     monospacedWidth=1
     size=15
     skewAmount=0
     typefaceFontName=Black
     outlineSettings {
      bApplyOutlineToDropShadows=false
      bMiteredCorners=false
      bSeparateFillAlpha=false
      outlineMaterial=null
      outlineSize=0
      outlineColor=<rgba 0 0 0 1>
     }
    }
   }
   slot {class=VerticalBoxSlot}
  }
  Text_FrontendXPLabel {
   widgetClassPath=&CommonTextBlock
   bIsVariable=true
   parent=VerticalBox_0
   widget=Text_FrontendXPLabel
   properties {
    justification=Center
    text=XP
    visibility=Collapsed
    colorAndOpacity {colorUseRule=UseColor_Specified, specifiedColor=<rgba 0.0003 0.0284 0.1683 1>}
    font {
     bForceMonospaced=false
     bMaterialIsStencil=false
     fontMaterial=null
     fontObject=&BurbankSmall
     letterSpacing=0
     monospacedWidth=1
     size=12
     skewAmount=0
     typefaceFontName=Black
     outlineSettings {
      bApplyOutlineToDropShadows=false
      bMiteredCorners=false
      bSeparateFillAlpha=false
      outlineMaterial=null
      outlineSize=0
      outlineColor=<rgba 0 0 0 1>
     }
    }
   }
   slot {class=VerticalBoxSlot, properties {padding=<ltrb 0 -7 0 0>}}
  }
  Border_Date {
   widgetClassPath=&Border
   bIsVariable=true
   parent=VerticalBox_1
   widget=Border_Date
   properties {
    horizontalAlignment=HAlign_Center
    visibility=SelfHitTestInvisible
    background {
     bIsDynamicallyLoaded=false
     drawAs=NoDrawType
     imageType=NoImage
     mirroring=NoMirror
     resourceName=None
     resourceObject=null
     tiling=NoTile
     imageSize=<32 32>
     margin=<ltrb 0 0 0 0>
     outlineSettings {
      bUseBrushTransparency=false
      roundingType=HalfHeightRadius
      width=0
      color {colorUseRule=UseColor_Specified, specifiedColor=<rgba 0 0 0 0>}
      cornerRadii=<0 0 0 0>
     }
     tintColor {colorUseRule=UseColor_Specified, specifiedColor=<rgba 1 1 1 1>}
     uVRegion {bIsValid=false, max=<0 0>, min=<0 0>}
    }
    padding=<ltrb 0 0 0 0>
   }
   slot {class=VerticalBoxSlot, properties {verticalAlignment=VAlign_Bottom, padding=<ltrb 0 0 0 6>}}
  }
  Text_UnlockDate {
   widgetClassPath=&CommonRichTextBlock
   bInherited=true
   bIsVariable=true
   parent=Border_Date
   widget=Text_UnlockDate
   properties {
    bOverrideDefaultStyle=true
    text=09/12
    defaultTextStyleOverride {
     overflowPolicy=Clip
     transformPolicy=None
     colorAndOpacity {colorUseRule=UseColor_Specified, specifiedColor=<rgba 1 0 0.0395 1>}
     font {
      bForceMonospaced=false
      bMaterialIsStencil=false
      fontMaterial=null
      fontObject=&BurbankSmall
      letterSpacing=0
      monospacedWidth=1
      size=12
      skewAmount=0
      typefaceFontName=Black
      outlineSettings {
       bApplyOutlineToDropShadows=false
       bMiteredCorners=false
       bSeparateFillAlpha=false
       outlineMaterial=null
       outlineSize=0
       outlineColor=<rgba 0 0 0 1>
      }
     }
     highlightColor {colorUseRule=UseColor_Specified, specifiedColor=<rgba 0 0 0 1>}
     highlightShape {
      bIsDynamicallyLoaded=false
      drawAs=Image
      imageType=NoImage
      mirroring=NoMirror
      resourceName=None
      resourceObject=null
      tiling=NoTile
      imageSize=<32 32>
      margin=<ltrb 0 0 0 0>
      outlineSettings {
       bUseBrushTransparency=false
       roundingType=HalfHeightRadius
       width=0
       color {colorUseRule=UseColor_Specified, specifiedColor=<rgba 0 0 0 0>}
       cornerRadii=<0 0 0 0>
      }
      tintColor {colorUseRule=UseColor_Specified, specifiedColor=<rgba 1 1 1 1>}
      uVRegion {bIsValid=false, max=<0 0>, min=<0 0>}
     }
     selectedBackgroundColor {colorUseRule=UseColor_Foreground, specifiedColor=<rgba 1 0 1 1>}
     shadowColorAndOpacity=<rgba 0 0 0 1>
     shadowOffset=<0 0>
     strikeBrush {
      bIsDynamicallyLoaded=false
      drawAs=Image
      imageType=NoImage
      mirroring=NoMirror
      resourceName=None
      resourceObject=null
      tiling=NoTile
      imageSize=<32 32>
      margin=<ltrb 0 0 0 0>
      outlineSettings {
       bUseBrushTransparency=false
       roundingType=HalfHeightRadius
       width=0
       color {colorUseRule=UseColor_Specified, specifiedColor=<rgba 0 0 0 0>}
       cornerRadii=<0 0 0 0>
      }
      tintColor {colorUseRule=UseColor_Specified, specifiedColor=<rgba 1 1 1 1>}
      uVRegion {bIsValid=false, max=<0 0>, min=<0 0>}
     }
     underlineBrush {
      bIsDynamicallyLoaded=false
      drawAs=Image
      imageType=NoImage
      mirroring=NoMirror
      resourceName=None
      resourceObject=null
      tiling=NoTile
      imageSize=<32 32>
      margin=<ltrb 0 0 0 0>
      outlineSettings {
       bUseBrushTransparency=false
       roundingType=HalfHeightRadius
       width=0
       color {colorUseRule=UseColor_Specified, specifiedColor=<rgba 0 0 0 0>}
       cornerRadii=<0 0 0 0>
      }
      tintColor {colorUseRule=UseColor_Specified, specifiedColor=<rgba 1 1 1 1>}
      uVRegion {bIsValid=false, max=<0 0>, min=<0 0>}
     }
    }
    renderTransform {angle=0, scale=<1 1>, shear=<-12 0>, translation=<0 0>}
   }
   slot {class=BorderSlot, properties {horizontalAlignment=HAlign_Center, padding=<ltrb 0 0 0 0>}}
  }
 ]
}
