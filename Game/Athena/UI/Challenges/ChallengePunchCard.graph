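# Widget Blueprint for a challenge punch card, built on AthenaChallengePunchCard, with a title, progress text, bang badge and punch slot entry box.
# Native-class events switch known and unknown faces, set progress, card number and badge, and apply style colors and lazy-loaded textures to the dynamic background material.
# Focus or non-touch hover plays a reversible animation, with sound and bang clearing.
# PreConstruct selects a smaller layout when IsIngame is set.
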WidgetBlueprint /Game/Athena/UI/Challenges/ChallengePunchCard
animations [hover {
  binding_count=8
  length_seconds=0.15
  object=&self:hover
  sequence {
   bindings [
    RichText_Progress {
     id=#1
     possessed_class=CommonRichTextBlock
     tracks [
      {
       type=MovieScene2DTransformTrack
       sections [{
         type=MovieScene2DTransformSection
         channels [
          Translation.X {type=MovieSceneScriptingFloatChannel}
          Translation.Y {type=MovieSceneScriptingFloatChannel}
          Angle {type=MovieSceneScriptingFloatChannel}
          Scale.X {type=MovieSceneScriptingFloatChannel, keys [[6,0.5],[18,1]]}
          Scale.Y {type=MovieSceneScriptingFloatChannel, keys [[6,0.5],[18,1]]}
          Shear.X {type=MovieSceneScriptingFloatChannel}
          Shear.Y {type=MovieSceneScriptingFloatChannel}
         ]
        }]
      }
      {
       type=MovieSceneFloatTrack
       display_name=Render Opacity
       sections [{type=MovieSceneFloatSection, channels [None {type=MovieSceneScriptingFloatChannel, keys [[6,0],[18,1]]}]}]
      }
     ]
    }
    Overlay_discovered {
     id=#2
     possessed_class=Overlay
     tracks [{
       type=MovieScene2DTransformTrack
       sections [{
         type=MovieScene2DTransformSection
         channels [
          Translation.X {type=MovieSceneScriptingFloatChannel}
          Translation.Y {type=MovieSceneScriptingFloatChannel}
          Angle {type=MovieSceneScriptingFloatChannel, keys [[0,0],[18,-1]]}
          Scale.X {type=MovieSceneScriptingFloatChannel}
          Scale.Y {type=MovieSceneScriptingFloatChannel}
          Shear.X {type=MovieSceneScriptingFloatChannel}
          Shear.Y {type=MovieSceneScriptingFloatChannel}
         ]
        }]
      }]
    }
    SizeBox_Header {
     id=#3
     possessed_class=SizeBox
     tracks [{
       type=MovieSceneFloatTrack
       display_name=Height Override
       sections [{type=MovieSceneFloatSection, channels [None {type=MovieSceneScriptingFloatChannel, keys [[0,116],[18,135]]}]}]
      }]
    }
    Border_Unknown {
     id=#4
     possessed_class=Border
     tracks [
      {
       type=MovieSceneWidgetMaterialTrack
       sections [{type=MovieSceneParameterSection, channels [hover {type=MovieSceneScriptingFloatChannel, keys [[0,0],[18,1]]}]}]
      }
      {
       type=MovieSceneColorTrack
       display_name=Brush Color
       sections [{
         type=MovieSceneColorSection
         end_frame=18
         start_frame=0
         channels [
          Color.R {type=MovieSceneScriptingFloatChannel}
          Color.G {type=MovieSceneScriptingFloatChannel}
          Color.B {type=MovieSceneScriptingFloatChannel}
          Color.A {type=MovieSceneScriptingFloatChannel, const [1,2]}
         ]
        }]
      }
      {
       type=MovieSceneColorTrack
       display_name=Content Color and Opacity
       sections [{
         type=MovieSceneColorSection
         end_frame=18
         start_frame=0
         channels [
          Color.R {type=MovieSceneScriptingFloatChannel}
          Color.G {type=MovieSceneScriptingFloatChannel}
          Color.B {type=MovieSceneScriptingFloatChannel}
          Color.A {type=MovieSceneScriptingFloatChannel, keys [[0,0.5],[18,1]]}
         ]
        }]
      }
     ]
    }
    Text_Description {
     id=#5
     possessed_class=CommonTextBlock
     tracks [{
       type=MovieScene2DTransformTrack
       sections [{
         type=MovieScene2DTransformSection
         channels [
          Translation.X {type=MovieSceneScriptingFloatChannel}
          Translation.Y {type=MovieSceneScriptingFloatChannel}
          Angle {type=MovieSceneScriptingFloatChannel}
          Scale.X {type=MovieSceneScriptingFloatChannel, keys [[0,0.5],[12,1]]}
          Scale.Y {type=MovieSceneScriptingFloatChannel, keys [[0,0.5],[12,1]]}
          Shear.X {type=MovieSceneScriptingFloatChannel}
          Shear.Y {type=MovieSceneScriptingFloatChannel}
         ]
        }]
      }]
    }
    FortLazyImage_Background {
     id=#6
     possessed_class=FortLazyImage
     tracks [{
       type=MovieSceneWidgetMaterialTrack
       sections [{
         type=MovieSceneParameterSection
         channels [
          hover {type=MovieSceneScriptingFloatChannel, keys [[0,0],[18,1]]}
          CenterAngle {type=MovieSceneScriptingFloatChannel, keys [[0,0.6],[18,1.4]]}
          CenterOffset {type=MovieSceneScriptingFloatChannel, keys [[0,0.55],[18,0.52]]}
         ]
        }]
      }]
    }
    Image_shadow {
     id=#7
     possessed_class=Image
     tracks [
      {
       type=MovieScene2DTransformTrack
       sections [{
         type=MovieScene2DTransformSection
         channels [
          Translation.X {type=MovieSceneScriptingFloatChannel, keys [[0,2],[18,6.7451]]}
          Translation.Y {type=MovieSceneScriptingFloatChannel, keys [[0,6],[18,4.9245]]}
          Angle {type=MovieSceneScriptingFloatChannel, keys [[0,0],[18,-0.03]]}
          Scale.X {type=MovieSceneScriptingFloatChannel}
          Scale.Y {type=MovieSceneScriptingFloatChannel}
          Shear.X {type=MovieSceneScriptingFloatChannel, keys [[0,0],[18,1.0591]]}
          Shear.Y {type=MovieSceneScriptingFloatChannel, const [0,1]}
         ]
        }]
      }
      {
       type=MovieSceneFloatTrack
       display_name=Render Opacity
       sections [{type=MovieSceneFloatSection, channels [None {type=MovieSceneScriptingFloatChannel, keys [[0,0.5],[18,0.75]]}]}]
      }
     ]
    }
    SizeBox_Infos {
     id=#8
     possessed_class=SizeBox
     tracks [
      {
       type=MovieSceneFloatTrack
       display_name=Height Override
       sections [{type=MovieSceneFloatSection, channels [None {type=MovieSceneScriptingFloatChannel, keys [[0,0],[18,56]]}]}]
      }
      {
       type=MovieSceneByteTrack
       display_name=Visibility
       sections [{type=MovieSceneByteSection, channels [None {type=MovieSceneScriptingByteChannel, keys [[0,1],[1,3]]}]}]
      }
      {
       type=MovieSceneFloatTrack
       display_name=Render Opacity
       sections [{type=MovieSceneFloatSection, end_frame=18, start_frame=0, channels [None {type=MovieSceneScriptingFloatChannel, keys [[0,0],[18,1]]}]}]
      }
     ]
    }
   ]
   display_rate {numerator=120}
  }
 }]
graphs [EventGraph {nodes [
   n1 = Break Fort Challenge Punch Card Styles!(FortChallengePunchCardStyles:"Fort Challenge Punch Card Styles Structure (by ref)") -> (
    EncodedName:String
    BaseColor1:Linear Color Structure
    BaseColor2:Linear Color Structure
    BaseColor3:Linear Color Structure
    BaseColor4:Linear Color Structure
    BaseColor5:Linear Color Structure
    AccentColor1:Linear Color Structure
    AccentColor2:Linear Color Structure
    BackgrounTexture:Texture 2D Soft Object Reference
   ) {is=title K2Node_BreakStruct}
   SetColorAndOpacity(
    self=Get Text_Description() -> (Text_Description:Common Text Object Reference)
    InColorAndOpacity=Make Slate Color(SpecifiedColor=@n1.BaseColor1)
   )
   > n2 = GetDynamicMaterial(self=Get FortLazyImage_Background() -> (FortLazyImage_Background:Fort Lazy Image Object Reference))
   > SetVectorParameterValue(self=@n2, ParameterName=BaseColor02, Value=@n1.BaseColor2)
   > SetVectorParameterValue(self=@n2, ParameterName=BaseColor03, Value=@n1.BaseColor3)
   > SetVectorParameterValue(self=@n2, ParameterName=BaseColor04, Value=@n1.BaseColor4)
   > SetDefaultColorAndOpacity!(
    execute
    self:Rich Text Block Object Reference=Get RichText_Progress() -> (RichText_Progress:Common Rich Text Block Object Reference)
    InColorAndOpacity:Slate Color Structure=Make Slate Color(SpecifiedColor=@n1.BaseColor4)
   ) -> (then)
   > n3 = Branch(Condition=IsAssetNull!(Asset:Soft Object Reference=@n1.BackgrounTexture) -> (ReturnValue:bool))
   n4 = Branch(execute=@n3.else, Condition=Get bSetLazyParamName() -> (bSetLazyParamName:bool))
   n5 = Get FortLazyImage_Background()
   SetBrushFromLazyTexture!(
    execute=[@n4.then, @n6.then]
    self:Common Lazy Image Object Reference=@n5
    LazyTexture:"Texture 2D Soft Object Reference (by ref)"=@n1.BackgrounTexture
    bMatchSize:bool=false
   ) -> (then)
   SetMaterialTextureParamName!(execute=@n4.else, self:Common Lazy Image Object Reference=@n5, TextureParamName:Name=Background) -> (then)
   > n6 = Set bSetLazyParamName(bSetLazyParamName:bool=true) -> (Output_Get:bool)
   n7 = Event SetUnkownState!() -> (OutputDelegate, then, bKnown:bool) {is=title K2Node_Event}
   > SetActiveWidgetIndex(
    self=Get WidgetSwitcher_punchcard() -> (WidgetSwitcher_punchcard:Widget Switcher Object Reference)
    Index=SelectInt(A=0, B=1, bPickA=@n7.bKnown)
   )
   n8 = Event SetProgressCount!() -> (OutputDelegate, then, AchivedCount:int=0, GoalCount:int=0) {is=title K2Node_Event}
   n9 = Get RichText_Progress()
   > SetText(
    self=@n9
    InText=Format Text!(Format="LOCTEXT(\"<text color=\\\"FFF\\\" size=\\\"24\\\">{Achived}</> / {Goal}\")", Achived:int=@n8.AchivedCount, Goal:int=@n8.GoalCount) -> (Result)
   )
   > SetVisibility(self=@n9, InVisibility=SelfHitTestInvisible)
   n10 = Event FocusPunchCard() {is=title K2Node_Event}
   n11 = Get hover() -> (hover:Widget Animation Object Reference)
   PlayAnimationForward(execute=[@n15.else, @n10.then], InAnimation=@n11)
   > n12 = IsAnimationPlayingForward!(execute, self:User Widget Object Reference, InAnimation:Widget Animation Object Reference=Get hover()) -> (then, ReturnValue:bool)
   > Branch(Condition=@n12)
   > PlaySound2D(Sound=/Game/Sounds/Fort_UI_Sounds/S13_Frontend/UI_Battlepass_Main_Hover_Cue)
   > TryToClearBangState(self:Athena Challenge Punch Card Object Reference)
   n13 = Event UnfocusPunchCard() {is=title K2Node_Event}
   PlayAnimationReverse(execute=[@n16.else, @n13.then], InAnimation=@n11)
   Event OnMouseEnter()
   n14 = IsUsingTouch(PlayerController=GetOwningPlayer())
   > n15 = Branch(Condition=@n14)
   Event OnMouseLeave!() -> (OutputDelegate, then, MouseEvent:"Pointer Event Structure (by ref)") {is=title K2Node_Event}
   > n16 = Branch(Condition=@n14)
   n17 = Event HideProgressCount!() -> (OutputDelegate, then, bBundleCompleted:bool) {is=title K2Node_Event}
   > n18 = Branch(Condition=@n17.bBundleCompleted)
   n19 = Get RichText_Progress()
   > SetText(self=@n19, InText=Format Text!(Format="LOCTEXT(\"COMPLETE!\")") -> (Result))
   > SetVisibility(self=@n19, InVisibility=SelfHitTestInvisible)
   SetVisibility(execute=@n18.else, self=@n19, InVisibility=Hidden)
   Event PreConstruct()
   n20 = Get IsIngame() -> (IsIngame:bool)
   > SetHeightOverride(
    self=Get SizeBox_Header() -> (SizeBox_Header:Size Box Object Reference)
    InHeightOverride=Select(Option 0:double=116, Option 1:double=80, Index=@n20) -> (ReturnValue:double)
   )
   > SetHeightOverride(
    self=Get SizeBox_punchslots() -> (SizeBox_punchslots:Size Box Object Reference)
    InHeightOverride=Select(Option 0=140, Option 1=90, Index=@n20)
   )
   > SetWidthOverride(
    self=Get SizeBox_Punchcard_Width() -> (SizeBox_Punchcard_Width:Size Box Object Reference)
    InWidthOverride=Select(Option 0=525, Option 1=450, Index=@n20)
   )
   n21 = Get FortLazyImage_Background()
   > SetMaterialTextureParamName(self=@n21, TextureParamName=Background)
   > n22 = GetDynamicMaterial(self=@n21)
   > SetScalarParameterValue(self=@n22, ParameterName=MakeOuterEdge, Value=Select(Option 0=1, Option 1=0, Index=@n20))
   > SetScalarParameterValue(self=@n22, ParameterName=CenterOffset, Value=Select(Option 0=0.55, Option 1=0.53, Index=@n20))
   > Set bSetLazyParamName(bSetLazyParamName=true)
   n23 = Get Text_Title() -> (Text_Title:Common Text Object Reference)
   n24 = Break Slate Font Info!(SlateFontInfo:"Slate Font Info Structure (by ref)"=Get Font!(self:Text Object Reference=@n23) -> (Font:Slate Font Info Structure)) -> (
    FontObject:Object Reference
    FontMaterial:Object Reference
    OutlineSettings:Font Outline Settings Structure
    TypefaceFontName:Name
    Size:float=~
    LetterSpacing:int
    SkewAmount:float=~
    bForceMonospaced:bool=~
    bMaterialIsStencil:bool=~
    MonospacedWidth:float=~
   ) {is=title K2Node_BreakStruct}
   n25 = Select(Option 0:int=26, Option 1:int=22, Index=@n20) -> (ReturnValue:int)
   > SetFont(
    self=@n23
    InFontInfo=MakeSlateFontInfo!(
     FontObject:Object Reference=@n24
     FontMaterial:Object Reference=@n24.FontMaterial
     OutlineSettings:Font Outline Settings Structure=@n24.OutlineSettings
     TypefaceFontName:Name=@n24.TypefaceFontName
     Size:float="To Float (Integer)"(InInt=@n25)
     LetterSpacing:int=0
     SkewAmount:float=0
     bForceMonospaced:bool=false
     bMaterialIsStencil:bool=false
     MonospacedWidth:float=1
    ) -> (ReturnValue:Slate Font Info Structure)
   )
   > SetVisibility(
    self=Get Image_shadow() -> (Image_shadow:Image Object Reference)
    InVisibility=Select(Option 0:ESlateVisibility Enum=SelfHitTestInvisible, Option 1:ESlateVisibility Enum=Collapsed, Index=@n20) -> (ReturnValue:ESlateVisibility Enum)
   )
   > SetHeightOverride(
    self=Get SizeBox_Infos() -> (SizeBox_Infos:Size Box Object Reference)
    InHeightOverride=Select(Option 0:double=0, Option 1:double=25, Index=@n20) -> (ReturnValue:double)
   )
   > SetVisibility(
    self=Get RichText_Progress()
    InVisibility=Select(Option 0:ESlateVisibility Enum=SelfHitTestInvisible, Option 1:ESlateVisibility Enum=Collapsed, Index=@n20) -> (ReturnValue:ESlateVisibility Enum)
   )
   n26 = Event SetUpEncodedName!() -> (OutputDelegate, then, Name:String, Count:int=0) {is=title K2Node_Event}
   n27 = Format Text!(
    Format="LOCTEXT(\"{Letter}-{Number}\")"
    Letter:String=@n26.Name
    Number:Text="To Text (Integer)"(Value=int + int(A=@n26.Count, B=1) -> (ReturnValue=0) {is=title K2Node_CommutativeAssociativeBinaryOperator}, MinimumIntegralDigits=2)
   ) -> (Result)
   > "SetText (Text)"(self=Get Text_CardNumber() -> (Text_CardNumber:Common Text Object Reference), InText=@n27)
   > "SetText (Text)"(self=Get Text_CardNumber_discovered() -> (Text_CardNumber_discovered:Common Text Object Reference), InText=@n27)
   n28 = Event SetBangState!() -> (OutputDelegate, then, bShowBang:bool) {is=title K2Node_Event}
   > n29 = Branch(Condition=@n28.bShowBang)
   n30 = Get Border_Bang() -> (Border_Bang:Border Object Reference)
   > SetVisibility(self=@n30, InVisibility=SelfHitTestInvisible)
   SetVisibility(execute=@n29.else, self=@n30, InVisibility=Hidden)
   Event SetStyle!() -> (OutputDelegate, then, PunchStyle:"Fort Challenge Punch Card Styles Structure (by ref)") {is=title K2Node_Event}
   Reroute Node(InputPin:Linear Color Structure=@n1.AccentColor2) -> (OutputPin:Linear Color Structure)
   Reroute Node(InputPin=@n1.AccentColor1)
   Reroute Node(InputPin=@n1.BaseColor5)
  ]}]
variables [
 AccentColor01 {type=LinearColor, default=<rgba 1 0.7379 0 1>}
 AccentColor02 {type=LinearColor, default=<rgba 1 0.402 0 1>}
 BaseColor01 {type=LinearColor, default=<rgba 0 1 1 1>}
 BaseColor02 {type=LinearColor, default=<rgba 0.0296 0.2542 0.7758 1>}
 BaseColor03 {type=LinearColor, default=<rgba 0.0086 0.0742 0.4735 1>}
 BaseColor04 {type=LinearColor, default=<rgba 0.0003 0.0194 0.1651 1>}
 BaseColor05 {type=LinearColor, default=<rgba 0 0.0037 0.0194 1>}
 bSetLazyParamName {type=boolean}
 IsIngame {type=boolean}
 NewVar_0 {type=integer}
]
widgets {
 info {inheritedWidgetCount=3, parentClass=/Script/FortniteUI.AthenaChallengePunchCard, rootWidgetClass=&SizeBox, widgetCount=40}
 widgets [
  SizeBox_Punchcard_Width {
   widgetClassPath=&SizeBox
   bIsVariable=true
   widget=SizeBox_Punchcard_Width
   properties {bOverride_WidthOverride=true, widthOverride=525}
  }
  WidgetSwitcher_punchcard {
   widgetClassPath=&WidgetSwitcher
   bIsVariable=true
   parent=SizeBox_Punchcard_Width
   widget=WidgetSwitcher_punchcard
   slot {class=SizeBoxSlot, properties {verticalAlignment=VAlign_Center}}
  }
  Overlay_discovered {widgetClassPath=&Overlay, parent=WidgetSwitcher_punchcard, widget=Overlay_discovered, slot {class=WidgetSwitcherSlot}}
  Image_shadow {
   widgetClassPath=&Image
   bIsVariable=true
   parent=Overlay_discovered
   widget=Image_shadow
   properties {
    renderOpacity=0.5
    brush {
     bIsDynamicallyLoaded=false
     drawAs=Box
     imageType=NoImage
     mirroring=NoMirror
     resourceName=None
     resourceObject=/Game/Athena/UI/Challenges/Art/M_Ui_Punchcard_Inst_Inst_Inst1
     tiling=NoTile
     imageSize=<256 256>
     margin=<ltrb 0.25 0.25 0.25 0.25>
     outlineSettings {
      bUseBrushTransparency=false
      roundingType=HalfHeightRadius
      width=0
      color {colorUseRule=UseColor_Specified, specifiedColor=<rgba 0 0 0 0>}
      cornerRadii=<0 0 0 0>
     }
     tintColor {colorUseRule=UseColor_Specified, specifiedColor=<rgba 0 0 0 1>}
     uVRegion {bIsValid=false, max=<0 0>, min=<0 0>}
    }
    renderTransform {angle=0, scale=<1.01 1.02>, shear=<0 0>, translation=<2 6>}
   }
   slot {class=OverlaySlot, properties {horizontalAlignment=HAlign_Fill, verticalAlignment=VAlign_Fill}}
  }
  FortLazyImage_Background {
   widgetClassPath=&FortLazyImage
   bIsVariable=true
   parent=Overlay_discovered
   widget=FortLazyImage_Background
   properties {brush {
     bIsDynamicallyLoaded=false
     drawAs=Box
     imageType=NoImage
     mirroring=NoMirror
     resourceName=None
     resourceObject=/Game/Athena/UI/Challenges/Art/M_Ui_Punchcard_Inst
     tiling=NoTile
     imageSize=<128 128>
     margin=<ltrb 0.2 0.2 0.2 0.2>
     outlineSettings {
      bUseBrushTransparency=false
      roundingType=HalfHeightRadius
      width=0
      color {colorUseRule=UseColor_Specified, specifiedColor=<rgba 0 0 0 0>}
      cornerRadii=<0 0 0 0>
     }
     tintColor {colorUseRule=UseColor_Specified, specifiedColor=<rgba 1 1 1 1>}
     uVRegion {bIsValid=false, max=<0 0>, min=<0 0>}
    }}
   slot {class=OverlaySlot, properties {horizontalAlignment=HAlign_Fill, verticalAlignment=VAlign_Fill}}
  }
  Border_Background {
   widgetClassPath=&Border
   bIsVariable=true
   parent=Overlay_discovered
   widget=Border_Background
   properties {
    visibility=SelfHitTestInvisible
    background {
     bIsDynamicallyLoaded=false
     drawAs=NoDrawType
     imageType=NoImage
     mirroring=NoMirror
     resourceName=None
     resourceObject=null
     tiling=NoTile
     imageSize=<32 32>
     margin=<ltrb 0.1 0.15 0.1 0.6>
     outlineSettings {
      bUseBrushTransparency=false
      roundingType=HalfHeightRadius
      width=0
      color {colorUseRule=UseColor_Specified, specifiedColor=<rgba 0 0 0 0>}
      cornerRadii=<0 0 0 0>
     }
     tintColor {colorUseRule=UseColor_Specified, specifiedColor=<rgba 1 1 1 1>}
     uVRegion {bIsValid=false, max=<0 0>, min=<0 0>}
    }
    padding=<ltrb 0 0 0 0>
   }
   slot {class=OverlaySlot, properties {horizontalAlignment=HAlign_Fill}}
  }
  VerticalBox_0 {
   widgetClassPath=&VerticalBox
   parent=Border_Background
   widget=VerticalBox_0
   slot {class=BorderSlot, properties {padding=<ltrb 0 0 0 0>}}
  }
  SizeBox_Header {
   widgetClassPath=&SizeBox
   bIsVariable=true
   parent=VerticalBox_0
   widget=SizeBox_Header
   properties {bOverride_HeightOverride=true, heightOverride=116, minDesiredHeight=96, widthOverride=596.7065}
   slot {class=VerticalBoxSlot, properties {padding=<ltrb 20 6 20 0>}}
  }
  Overlay_1 {widgetClassPath=&Overlay, parent=SizeBox_Header, widget=Overlay_1, slot {class=SizeBoxSlot}}
  Text_CardNumber_discovered {
   widgetClassPath=&CommonTextBlock
   bIsVariable=true
   parent=Overlay_1
   widget=Text_CardNumber_discovered
   properties {
    justification=Center
    lineHeightPercentage=0.9
    scrollStyle=&TextScrollStyle-Base_C
    text=A-01
    textTransformPolicy=ToUpper
    colorAndOpacity {colorUseRule=UseColor_Specified, specifiedColor=<rgba 1 0 0.0296 0.25>}
    font {
     bForceMonospaced=false
     bMaterialIsStencil=false
     fontMaterial=null
     fontObject=&BurbankSmall
     letterSpacing=0
     monospacedWidth=1
     size=20
     skewAmount=0
     typefaceFontName=Black
     outlineSettings {
      bApplyOutlineToDropShadows=false
      bMiteredCorners=false
      bSeparateFillAlpha=false
      outlineMaterial=null
      outlineSize=0
      outlineColor=<rgba 0 0 0 1>
     }
    }
    shadowOffset=<0 0>
   }
   slot {class=OverlaySlot, properties {padding=<ltrb 0 6 0 0>}}
  }
  ScaleBox_0 {
   widgetClassPath=&ScaleBox
   parent=Overlay_1
   widget=ScaleBox_0
   properties {stretch=ScaleToFitY, stretchDirection=DownOnly}
   slot {class=OverlaySlot, properties {horizontalAlignment=HAlign_Fill, verticalAlignment=VAlign_Fill, padding=<ltrb 6 12 6 12>}}
  }
  Image_192 {
   widgetClassPath=&Image
   bIsVariable=true
   parent=ScaleBox_0
   widget=Image_192
   properties {
    visibility=SelfHitTestInvisible
    brush {
     bIsDynamicallyLoaded=false
     drawAs=Image
     imageType=NoImage
     mirroring=NoMirror
     resourceName=None
     resourceObject=/Game/Athena/UI/Challenges/Art/T_Ui_Hightower_Logo_128_Mat
     tiling=NoTile
     imageSize=<90 90>
     margin=<ltrb 0 0 0 0>
     outlineSettings {
      bUseBrushTransparency=false
      roundingType=HalfHeightRadius
      width=0
      color {colorUseRule=UseColor_Specified, specifiedColor=<rgba 0 0 0 0>}
      cornerRadii=<0 0 0 0>
     }
     tintColor {colorUseRule=UseColor_Specified, specifiedColor=<rgba 0.0262 0 0.0097 1>}
     uVRegion {bIsValid=false, max=<0 0>, min=<0 0>}
    }
    colorAndOpacity=<rgba 1 1 1 0.1048>
   }
   slot {class=ScaleBoxSlot, properties {horizontalAlignment=HAlign_Right}}
  }
  VerticalBox_1 {
   widgetClassPath=&VerticalBox
   parent=Overlay_1
   widget=VerticalBox_1
   slot {class=OverlaySlot, properties {horizontalAlignment=HAlign_Fill, verticalAlignment=VAlign_Center}}
  }
  Text_Title {
   widgetClassPath=&CommonTextBlock
   bInherited=true
   bIsVariable=true
   parent=VerticalBox_1
   widget=Text_Title
   properties {
    justification=Center
    lineHeightPercentage=0.9
    scrollStyle=&TextScrollStyle-Base_C
    text=Played Close To The Chest
    textTransformPolicy=ToUpper
    font {
     bForceMonospaced=false
     bMaterialIsStencil=false
     fontMaterial=null
     fontObject=&Burbank
     letterSpacing=0
     monospacedWidth=1
     size=26
     skewAmount=0
     typefaceFontName=Burbank-PlayerPower
     outlineSettings {
      bApplyOutlineToDropShadows=false
      bMiteredCorners=false
      bSeparateFillAlpha=false
      outlineMaterial=null
      outlineSize=0
      outlineColor=<rgba 0 0 0 1>
     }
    }
    renderTransform {angle=0, scale=<1 1>, shear=<-12 0>, translation=<0 0>}
    shadowOffset=<0 0>
   }
   slot {class=VerticalBoxSlot, properties {padding=<ltrb 0 1 0 2>}}
  }
  SizeBox_Infos {
   widgetClassPath=&SizeBox
   bIsVariable=true
   parent=VerticalBox_1
   widget=SizeBox_Infos
   properties {bOverride_HeightOverride=true, visibility=Collapsed}
   slot {class=VerticalBoxSlot}
  }
  VerticalBox_4 {
   widgetClassPath=&VerticalBox
   parent=SizeBox_Infos
   widget=VerticalBox_4
   slot {class=SizeBoxSlot, properties {verticalAlignment=VAlign_Bottom}}
  }
  Text_Description {
   widgetClassPath=&CommonTextBlock
   bInherited=true
   parent=VerticalBox_4
   widget=Text_Description
   properties {
    justification=Center
    scrollStyle=&TextScrollStyle-Base_C
    text=Open Chests
    colorAndOpacity {colorUseRule=UseColor_Specified, specifiedColor=<rgba 1 0.6584 0.0356 1>}
    font {
     bForceMonospaced=false
     bMaterialIsStencil=false
     fontMaterial=null
     fontObject=&BurbankSmall
     letterSpacing=0
     monospacedWidth=1
     size=15
     skewAmount=0
     typefaceFontName=Bold
     outlineSettings {
      bApplyOutlineToDropShadows=false
      bMiteredCorners=false
      bSeparateFillAlpha=false
      outlineMaterial=null
      outlineSize=0
      outlineColor=<rgba 0 0 0 1>
     }
    }
    renderTransform {angle=0, scale=<1 1>, shear=<-12 0>, translation=<0 0>}
    shadowOffset=<0 0>
   }
   slot {class=VerticalBoxSlot}
  }
  RichText_Progress {
   widgetClassPath=&CommonRichTextBlock
   bIsVariable=true
   parent=VerticalBox_4
   widget=RichText_Progress
   properties {
    bOverrideDefaultStyle=true
    justification=Center
    text="<text color=\"FFF\" size=\"24\">7</> / 12"
    defaultTextStyleOverride {
     overflowPolicy=Clip
     transformPolicy=None
     colorAndOpacity {colorUseRule=UseColor_Specified, specifiedColor=<rgba 0.0802 0 0.0123 1>}
     font {
      bForceMonospaced=false
      bMaterialIsStencil=false
      fontMaterial=null
      fontObject=&BurbankSmall
      letterSpacing=0
      monospacedWidth=1
      size=20
      skewAmount=0
      typefaceFontName=Black
      outlineSettings {
       bApplyOutlineToDropShadows=false
       bMiteredCorners=false
       bSeparateFillAlpha=false
       outlineMaterial=null
       outlineSize=0
       outlineColor=<rgba 0 0 0 1>
      }
     }
     highlightColor {colorUseRule=UseColor_Specified, specifiedColor=<rgba 0 0 0 1>}
     highlightShape {
      bIsDynamicallyLoaded=false
      drawAs=Image
      imageType=NoImage
      mirroring=NoMirror
      resourceName=None
      resourceObject=null
      tiling=NoTile
      imageSize=<32 32>
      margin=<ltrb 0 0 0 0>
      outlineSettings {
       bUseBrushTransparency=false
       roundingType=HalfHeightRadius
       width=0
       color {colorUseRule=UseColor_Specified, specifiedColor=<rgba 0 0 0 0>}
       cornerRadii=<0 0 0 0>
      }
      tintColor {colorUseRule=UseColor_Specified, specifiedColor=<rgba 1 1 1 1>}
      uVRegion {bIsValid=false, max=<0 0>, min=<0 0>}
     }
     selectedBackgroundColor {colorUseRule=UseColor_Foreground, specifiedColor=<rgba 1 0 1 1>}
     shadowColorAndOpacity=<rgba 0 0 0 1>
     shadowOffset=<0 0>
     strikeBrush {
      bIsDynamicallyLoaded=false
      drawAs=Image
      imageType=NoImage
      mirroring=NoMirror
      resourceName=None
      resourceObject=null
      tiling=NoTile
      imageSize=<32 32>
      margin=<ltrb 0 0 0 0>
      outlineSettings {
       bUseBrushTransparency=false
       roundingType=HalfHeightRadius
       width=0
       color {colorUseRule=UseColor_Specified, specifiedColor=<rgba 0 0 0 0>}
       cornerRadii=<0 0 0 0>
      }
      tintColor {colorUseRule=UseColor_Specified, specifiedColor=<rgba 1 1 1 1>}
      uVRegion {bIsValid=false, max=<0 0>, min=<0 0>}
     }
     underlineBrush {
      bIsDynamicallyLoaded=false
      drawAs=Image
      imageType=NoImage
      mirroring=NoMirror
      resourceName=None
      resourceObject=null
      tiling=NoTile
      imageSize=<32 32>
      margin=<ltrb 0 0 0 0>
      outlineSettings {
       bUseBrushTransparency=false
       roundingType=HalfHeightRadius
       width=0
       color {colorUseRule=UseColor_Specified, specifiedColor=<rgba 0 0 0 0>}
       cornerRadii=<0 0 0 0>
      }
      tintColor {colorUseRule=UseColor_Specified, specifiedColor=<rgba 1 1 1 1>}
      uVRegion {bIsValid=false, max=<0 0>, min=<0 0>}
     }
    }
    margin=<ltrb 0 -4 0 0>
    renderTransform {angle=0, scale=<1 1>, shear=<-12 0>, translation=<0 0>}
   }
   slot {class=VerticalBoxSlot, properties {padding=<ltrb 0 2 0 0>}}
  }
  SizeBox_punchslots {
   widgetClassPath=&SizeBox
   bIsVariable=true
   parent=VerticalBox_0
   widget=SizeBox_punchslots
   properties {bOverride_HeightOverride=true, heightOverride=140}
   slot {class=VerticalBoxSlot, properties {padding=<ltrb 16 1 16 8>}}
  }
  Box_PunchSlots {
   widgetClassPath=&DynamicEntryBox
   bInherited=true
   bIsVariable=true
   parent=SizeBox_punchslots
   widget=Box_PunchSlots
   properties {
    entryWidgetClass=&/ChallengePunchCardEntry.ChallengePunchCardEntry_C
    numDesignerPreviewEntries=4
    entrySizeRule {sizeRule=Fill, value=1}
   }
   slot {class=SizeBoxSlot}
  }
  HorizontalBox_90 {
   widgetClassPath=&HorizontalBox
   parent=Overlay_discovered
   widget=HorizontalBox_90
   properties {visibility=Collapsed}
   slot {class=OverlaySlot, properties {horizontalAlignment=HAlign_Right, verticalAlignment=VAlign_Bottom, padding=<ltrb 0 0 0 -12>}}
  }
  base1 {
   widgetClassPath=&Image
   bIsVariable=true
   parent=HorizontalBox_90
   widget=base1
   properties {
    visibility=SelfHitTestInvisible
    brush {
     bIsDynamicallyLoaded=false
     drawAs=Image
     imageType=NoImage
     mirroring=NoMirror
     resourceName=None
     resourceObject=null
     tiling=NoTile
     imageSize=<8 8>
     margin=<ltrb 0 0 0 0>
     outlineSettings {
      bUseBrushTransparency=false
      roundingType=HalfHeightRadius
      width=0
      color {colorUseRule=UseColor_Specified, specifiedColor=<rgba 0 0 0 0>}
      cornerRadii=<0 0 0 0>
     }
     tintColor {colorUseRule=UseColor_Specified, specifiedColor=<rgba 1 1 1 1>}
     uVRegion {bIsValid=false, max=<0 0>, min=<0 0>}
    }
   }
   slot {class=HorizontalBoxSlot, properties {padding=<ltrb 4 0 0 0>}}
  }
  base2 {
   widgetClassPath=&Image
   bIsVariable=true
   parent=HorizontalBox_90
   widget=base2
   properties {
    visibility=SelfHitTestInvisible
    brush {
     bIsDynamicallyLoaded=false
     drawAs=Image
     imageType=NoImage
     mirroring=NoMirror
     resourceName=None
     resourceObject=null
     tiling=NoTile
     imageSize=<8 8>
     margin=<ltrb 0 0 0 0>
     outlineSettings {
      bUseBrushTransparency=false
      roundingType=HalfHeightRadius
      width=0
      color {colorUseRule=UseColor_Specified, specifiedColor=<rgba 0 0 0 0>}
      cornerRadii=<0 0 0 0>
     }
     tintColor {colorUseRule=UseColor_Specified, specifiedColor=<rgba 1 1 1 1>}
     uVRegion {bIsValid=false, max=<0 0>, min=<0 0>}
    }
   }
   slot {class=HorizontalBoxSlot, properties {padding=<ltrb 4 0 0 0>}}
  }
  base3 {
   widgetClassPath=&Image
   bIsVariable=true
   parent=HorizontalBox_90
   widget=base3
   properties {
    visibility=SelfHitTestInvisible
    brush {
     bIsDynamicallyLoaded=false
     drawAs=Image
     imageType=NoImage
     mirroring=NoMirror
     resourceName=None
     resourceObject=null
     tiling=NoTile
     imageSize=<8 8>
     margin=<ltrb 0 0 0 0>
     outlineSettings {
      bUseBrushTransparency=false
      roundingType=HalfHeightRadius
      width=0
      color {colorUseRule=UseColor_Specified, specifiedColor=<rgba 0 0 0 0>}
      cornerRadii=<0 0 0 0>
     }
     tintColor {colorUseRule=UseColor_Specified, specifiedColor=<rgba 1 1 1 1>}
     uVRegion {bIsValid=false, max=<0 0>, min=<0 0>}
    }
   }
   slot {class=HorizontalBoxSlot, properties {padding=<ltrb 4 0 0 0>}}
  }
  base4 {
   widgetClassPath=&Image
   bIsVariable=true
   parent=HorizontalBox_90
   widget=base4
   properties {
    visibility=SelfHitTestInvisible
    brush {
     bIsDynamicallyLoaded=false
     drawAs=Image
     imageType=NoImage
     mirroring=NoMirror
     resourceName=None
     resourceObject=null
     tiling=NoTile
     imageSize=<8 8>
     margin=<ltrb 0 0 0 0>
     outlineSettings {
      bUseBrushTransparency=false
      roundingType=HalfHeightRadius
      width=0
      color {colorUseRule=UseColor_Specified, specifiedColor=<rgba 0 0 0 0>}
      cornerRadii=<0 0 0 0>
     }
     tintColor {colorUseRule=UseColor_Specified, specifiedColor=<rgba 1 1 1 1>}
     uVRegion {bIsValid=false, max=<0 0>, min=<0 0>}
    }
   }
   slot {class=HorizontalBoxSlot, properties {padding=<ltrb 4 0 0 0>}}
  }
  base5 {
   widgetClassPath=&Image
   bIsVariable=true
   parent=HorizontalBox_90
   widget=base5
   properties {
    visibility=SelfHitTestInvisible
    brush {
     bIsDynamicallyLoaded=false
     drawAs=Image
     imageType=NoImage
     mirroring=NoMirror
     resourceName=None
     resourceObject=null
     tiling=NoTile
     imageSize=<8 8>
     margin=<ltrb 0 0 0 0>
     outlineSettings {
      bUseBrushTransparency=false
      roundingType=HalfHeightRadius
      width=0
      color {colorUseRule=UseColor_Specified, specifiedColor=<rgba 0 0 0 0>}
      cornerRadii=<0 0 0 0>
     }
     tintColor {colorUseRule=UseColor_Specified, specifiedColor=<rgba 1 1 1 1>}
     uVRegion {bIsValid=false, max=<0 0>, min=<0 0>}
    }
   }
   slot {class=HorizontalBoxSlot, properties {padding=<ltrb 4 0 0 0>}}
  }
  Accent1 {
   widgetClassPath=&Image
   bIsVariable=true
   parent=HorizontalBox_90
   widget=Accent1
   properties {
    visibility=SelfHitTestInvisible
    brush {
     bIsDynamicallyLoaded=false
     drawAs=Image
     imageType=NoImage
     mirroring=NoMirror
     resourceName=None
     resourceObject=null
     tiling=NoTile
     imageSize=<8 8>
     margin=<ltrb 0 0 0 0>
     outlineSettings {
      bUseBrushTransparency=false
      roundingType=HalfHeightRadius
      width=0
      color {colorUseRule=UseColor_Specified, specifiedColor=<rgba 0 0 0 0>}
      cornerRadii=<0 0 0 0>
     }
     tintColor {colorUseRule=UseColor_Specified, specifiedColor=<rgba 1 1 1 1>}
     uVRegion {bIsValid=false, max=<0 0>, min=<0 0>}
    }
   }
   slot {class=HorizontalBoxSlot, properties {padding=<ltrb 4 0 0 0>}}
  }
  Accent2 {
   widgetClassPath=&Image
   bIsVariable=true
   parent=HorizontalBox_90
   widget=Accent2
   properties {
    visibility=SelfHitTestInvisible
    brush {
     bIsDynamicallyLoaded=false
     drawAs=Image
     imageType=NoImage
     mirroring=NoMirror
     resourceName=None
     resourceObject=null
     tiling=NoTile
     imageSize=<8 8>
     margin=<ltrb 0 0 0 0>
     outlineSettings {
      bUseBrushTransparency=false
      roundingType=HalfHeightRadius
      width=0
      color {colorUseRule=UseColor_Specified, specifiedColor=<rgba 0 0 0 0>}
      cornerRadii=<0 0 0 0>
     }
     tintColor {colorUseRule=UseColor_Specified, specifiedColor=<rgba 1 1 1 1>}
     uVRegion {bIsValid=false, max=<0 0>, min=<0 0>}
    }
   }
   slot {class=HorizontalBoxSlot, properties {padding=<ltrb 4 0 0 0>}}
  }
  Border_Bang {
   widgetClassPath=&Border
   bIsVariable=true
   parent=Overlay_discovered
   widget=Border_Bang
   properties {
    visibility=Hidden
    background {
     bIsDynamicallyLoaded=false
     drawAs=Box
     imageType=NoImage
     mirroring=NoMirror
     resourceName=None
     resourceObject=/Game/Athena/UI/ReplayBrowser/M_Ui_Skewed_Vfor9slices_Inst
     tiling=NoTile
     imageSize=<16 16>
     margin=<ltrb 0 0 0 0>
     outlineSettings {
      bUseBrushTransparency=false
      roundingType=HalfHeightRadius
      width=0
      color {colorUseRule=UseColor_Specified, specifiedColor=<rgba 0 0 0 0>}
      cornerRadii=<0 0 0 0>
     }
     tintColor {colorUseRule=UseColor_Specified, specifiedColor=<rgba 1 0.6584 0.0356 1>}
     uVRegion {bIsValid=false, max=<0 0>, min=<0 0>}
    }
    padding=<ltrb 10 0 8 0>
    renderTransform {angle=0, scale=<1 1>, shear=<-12 0>, translation=<0 0>}
   }
   slot {class=OverlaySlot, properties {horizontalAlignment=HAlign_Right, padding=<ltrb 0 -6 12 0>}}
  }
  CommonTextBlock_130 {
   widgetClassPath=&CommonTextBlock
   parent=Border_Bang
   widget=CommonTextBlock_130
   properties {
    text="!"
    colorAndOpacity {colorUseRule=UseColor_Specified, specifiedColor=<rgba 0 0 0 1>}
    font {
     bForceMonospaced=false
     bMaterialIsStencil=false
     fontMaterial=null
     fontObject=&Burbank
     letterSpacing=0
     monospacedWidth=1
     size=24
     skewAmount=0
     typefaceFontName=Burbank-Black
     outlineSettings {
      bApplyOutlineToDropShadows=false
      bMiteredCorners=false
      bSeparateFillAlpha=false
      outlineMaterial=null
      outlineSize=0
      outlineColor=<rgba 0 0 0 1>
     }
    }
    shadowOffset=<0 0>
   }
   slot {class=BorderSlot, properties {padding=<ltrb 10 0 8 0>}}
  }
  Overlay_unknown {
   widgetClassPath=&Overlay
   bIsVariable=true
   parent=WidgetSwitcher_punchcard
   widget=Overlay_unknown
   slot {class=WidgetSwitcherSlot}
  }
  Border_Unknown {
   widgetClassPath=&Border
   bIsVariable=true
   parent=Overlay_unknown
   widget=Border_Unknown
   properties {
    visibility=SelfHitTestInvisible
    background {
     bIsDynamicallyLoaded=false
     drawAs=Box
     imageType=NoImage
     mirroring=NoMirror
     resourceName=None
     resourceObject=/Game/Athena/UI/Challenges/Art/M_Ui_Punchcard_Inst_Inst
     tiling=NoTile
     imageSize=<256 256>
     margin=<ltrb 0.1 0.15 0.1 0.6>
     outlineSettings {
      bUseBrushTransparency=false
      roundingType=HalfHeightRadius
      width=0
      color {colorUseRule=UseColor_Specified, specifiedColor=<rgba 0 0 0 0>}
      cornerRadii=<0 0 0 0>
     }
     tintColor {colorUseRule=UseColor_Specified, specifiedColor=<rgba 0.0262 0 0.0097 1>}
     uVRegion {bIsValid=false, max=<0 0>, min=<0 0>}
    }
    contentColorAndOpacity=<rgba 1 1 1 0.498>
    padding=<ltrb 0 0 0 0>
   }
   slot {class=OverlaySlot, properties {horizontalAlignment=HAlign_Fill}}
  }
  SizeBox_81 {
   widgetClassPath=&SizeBox
   parent=Border_Unknown
   widget=SizeBox_81
   properties {bOverride_HeightOverride=true, heightOverride=268}
   slot {class=BorderSlot, properties {padding=<ltrb 0 0 0 0>}}
  }
  Border_logo {
   widgetClassPath=&Border
   parent=SizeBox_81
   widget=Border_logo
   properties {
    visibility=SelfHitTestInvisible
    background {
     bIsDynamicallyLoaded=false
     drawAs=NoDrawType
     imageType=NoImage
     mirroring=NoMirror
     resourceName=None
     resourceObject=null
     tiling=NoTile
     imageSize=<32 32>
     margin=<ltrb 0 0 0 0>
     outlineSettings {
      bUseBrushTransparency=false
      roundingType=HalfHeightRadius
      width=0
      color {colorUseRule=UseColor_Specified, specifiedColor=<rgba 0 0 0 0>}
      cornerRadii=<0 0 0 0>
     }
     tintColor {colorUseRule=UseColor_Specified, specifiedColor=<rgba 1 1 1 1>}
     uVRegion {bIsValid=false, max=<0 0>, min=<0 0>}
    }
    padding=<ltrb 0 0 0 0>
   }
   slot {class=SizeBoxSlot}
  }
  Overlay_0 {widgetClassPath=&Overlay, parent=Border_logo, widget=Overlay_0, slot {class=BorderSlot, properties {padding=<ltrb 0 0 0 0>}}}
  Image_FortniteLogo {
   widgetClassPath=&Image
   parent=Overlay_0
   widget=Image_FortniteLogo
   properties {
    visibility=HitTestInvisible
    brush {
     bIsDynamicallyLoaded=false
     drawAs=Image
     imageType=NoImage
     mirroring=NoMirror
     resourceName=None
     resourceObject=&T_Ui_Hightower_Logo_128_Mat
     tiling=NoTile
     imageSize=<128 128>
     margin=<ltrb 0 0 0 0>
     outlineSettings {
      bUseBrushTransparency=false
      roundingType=HalfHeightRadius
      width=0
      color {colorUseRule=UseColor_Specified, specifiedColor=<rgba 0 0 0 0>}
      cornerRadii=<0 0 0 0>
     }
     tintColor {colorUseRule=UseColor_Specified, specifiedColor=<rgba 0.0262 0 0.0097 0.75>}
     uVRegion {bIsValid=false, max=<0 0>, min=<0 0>}
    }
   }
   slot {class=OverlaySlot, properties {horizontalAlignment=HAlign_Center, verticalAlignment=VAlign_Center}}
  }
  HorizontalBox_0 {
   widgetClassPath=&HorizontalBox
   parent=Overlay_0
   widget=HorizontalBox_0
   properties {visibility=Collapsed}
   slot {class=OverlaySlot, properties {verticalAlignment=VAlign_Bottom, padding=<ltrb 0 0 0 12>}}
  }
  Image_151 {
   widgetClassPath=&Image
   bIsVariable=true
   parent=HorizontalBox_0
   widget=Image_151
   properties {
    visibility=SelfHitTestInvisible
    brush {
     bIsDynamicallyLoaded=false
     drawAs=Box
     imageType=NoImage
     mirroring=NoMirror
     resourceName=None
     resourceObject=/Game/Athena/UI/ReplayBrowser/M_Ui_Skewed_Vfor9slices_Inst2
     tiling=NoTile
     imageSize=<8 8>
     margin=<ltrb 0.42 0.42 0.58 0.58>
     outlineSettings {
      bUseBrushTransparency=false
      roundingType=HalfHeightRadius
      width=0
      color {colorUseRule=UseColor_Specified, specifiedColor=<rgba 0 0 0 0>}
      cornerRadii=<0 0 0 0>
     }
     tintColor {colorUseRule=UseColor_Specified, specifiedColor=<rgba 1 1 1 1>}
     uVRegion {bIsValid=false, max=<0 0>, min=<0 0>}
    }
    renderTransform {angle=0, scale=<-1 1>, shear=<0 0>, translation=<0 0>}
   }
   slot {class=HorizontalBoxSlot, properties {verticalAlignment=VAlign_Center, size {sizeRule=Fill, value=1}}}
  }
  Image {
   widgetClassPath=&Image
   bIsVariable=true
   parent=HorizontalBox_0
   widget=Image
   properties {
    visibility=SelfHitTestInvisible
    brush {
     bIsDynamicallyLoaded=false
     drawAs=Box
     imageType=NoImage
     mirroring=NoMirror
     resourceName=None
     resourceObject=&M_Ui_Skewed_Vfor9slices_Inst2
     tiling=NoTile
     imageSize=<8 8>
     margin=<ltrb 0.5 0.5 0.5 0.5>
     outlineSettings {
      bUseBrushTransparency=false
      roundingType=HalfHeightRadius
      width=0
      color {colorUseRule=UseColor_Specified, specifiedColor=<rgba 0 0 0 0>}
      cornerRadii=<0 0 0 0>
     }
     tintColor {colorUseRule=UseColor_Specified, specifiedColor=<rgba 1 1 1 1>}
     uVRegion {bIsValid=false, max=<0 0>, min=<0 0>}
    }
   }
   slot {class=HorizontalBoxSlot, properties {verticalAlignment=VAlign_Center, size {sizeRule=Fill, value=1}}}
  }
  Text_CardNumber {
   widgetClassPath=&CommonTextBlock
   bIsVariable=true
   parent=Overlay_0
   widget=Text_CardNumber
   properties {
    justification=Center
    lineHeightPercentage=0.9
    scrollStyle=&TextScrollStyle-Base_C
    text=W-01
    textTransformPolicy=ToUpper
    colorAndOpacity {colorUseRule=UseColor_Specified, specifiedColor=<rgba 1 0 0.0395 1>}
    font {
     bForceMonospaced=false
     bMaterialIsStencil=false
     fontMaterial=null
     fontObject=&BurbankSmall
     letterSpacing=0
     monospacedWidth=1
     size=20
     skewAmount=0
     typefaceFontName=Black
     outlineSettings {
      bApplyOutlineToDropShadows=false
      bMiteredCorners=false
      bSeparateFillAlpha=false
      outlineMaterial=null
      outlineSize=0
      outlineColor=<rgba 0 0 0 1>
     }
    }
    shadowOffset=<0 0>
   }
   slot {class=OverlaySlot, properties {padding=<ltrb 16 7 24 12>}}
  }
 ]
}
